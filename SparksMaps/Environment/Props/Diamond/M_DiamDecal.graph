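# Deferred decal material for the Diamond environment prop, projecting an animated, tinted texture pattern with alpha-composite blending.
# Scaled, rotated and panned UVs drive offset texture samples packed into RGB for fake chromatic aberration, blended toward a single-channel sample.
# Radial sphere masks shape opacity and an optional fake shadow, and material outputs are premultiplied for compositing.
# A switch limits emissive to night via time-of-day, and tint comes from parameter or vertex color.
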Material /SparksMaps/Environment/Props/Diamond/M_DiamDecal
graph {
 properties {
  bUseMaterialAttributes=true
  bUsedWithNanite=true
  blendMode=BLEND_AlphaComposite
  materialDomain=MD_DeferredDecal
  pixelDepthOffsetMode=PDOM_Legacy
  usedShadingModels=Default Lit
 }
 nodes [
  n1 = TextureCoordinate()
  n2 = ScalarParameter() {defaultValue=1, group=Animation, parameterName=TextureScaleUniform}
  n3 = VectorParameter() {group=Animation, parameterName=Texture_ScaleOffset, defaultValue=<rgba 1 1 0 0>}
  n4 = AppendVector(A=@n3.R, B=@n3.R)
  n5 = Add(A=Multiply(A=ConstantBiasScale(None=@n1) {bias=-0.5, scale=2}, B=Multiply(A=@n2, B=@n4)), B=AppendVector(A=@n3.B, B=@n3.B))
  n6 = VectorParameter() {group=Animation, parameterName=PanSpeed_RotSpeed}
  n7 = Multiply(A=@n6.B, B=Time())
  n8 = Panner(Coordinate=ConstantBiasScale(None=Rotator(Coordinate=@n5, Time=@n7)), Speed=AppendVector(A=@n6.R, B=@n6.R)) {speedX=0.1}
  n9 = TextureObjectParameter() {parameterName=Decal_Texture, texture=/Game/Global/Textures/Default/Masks_and_Gradients/T_TrianglePattern}
  n10 = TextureSample(UVs=@n8, Tex=@n9)
  n11 = ScalarParameter() {defaultValue=1, group=Animation, parameterName=ChromaticAb_Strength}
  n12 = VectorParameter() {group=Animation, parameterName=Texture_RGB_Offsets, defaultValue=<rgba 0.05 0.05 -0.05 -0.05>}
  n13 = Multiply(A=@n11, B=AppendVector(A=@n12.R, B=@n12.R))
  n14 = Multiply(A=@n11, B=AppendVector(A=@n12.B, B=@n12.B))
  n15 = MaterialFunctionCall(R=@n10.R, G=TextureSample(UVs=Add(A=@n8, B=@n13), Tex=@n9).R, B=TextureSample(UVs=Add(A=@n8, B=@n14), Tex=@n9).R) {materialFunction=&AppendMany}
  n16 = ScalarParameter() {defaultValue=1, group=Animation, parameterName=Texture_Hardness}
  n17 = NamedRerouteUsage()
  n18 = NamedRerouteUsage()
  n19 = LinearInterpolate(
   A=LinearInterpolate(A=Multiply(A=@n15.RGB) {constB=0.85}, B=Power(Base=@n15.RGB, Exp=@n16), Alpha=@n17)
   B=LinearInterpolate(A=Multiply(A=@n10.R) {constB=0.85}, B=Power(Base=@n10.R, Exp=@n16), Alpha=@n17)
   Alpha=Power(Base=Add(A=@n18) {constB=-0.075}, Exp=ScalarParameter() {defaultValue=0.85, group=Animation, parameterName=ChromaticAb_Pow})
  )
  Multiply(A=@n19, B=NamedRerouteUsage())
  n20 = TextureCoordinate()
  n21 = Constant() {r=0.5}
  n22 = ScalarParameter() {defaultValue=0.35, group=Animation, parameterName=FakeShadow_Radius}
  n23 = ScalarParameter() {defaultValue=100, group=Animation, parameterName=FakeShadow_Hardness}
  n24 = LinearInterpolate(
   B=OneMinus(None=SphereMask(A=@n20, B=@n21, Radius=@n22, Hardness=@n23) {attenuationRadius=0.5, hardnessPercent=90})
   Alpha=ScalarParameter() {group=Animation, parameterName=FakeShadow_Strength}
  ) {constA=1}
  n25 = SphereMask(
   A=TextureCoordinate()
   B=Constant() {r=0.5}
   Radius=ScalarParameter() {defaultValue=0.5, group=Animation, parameterName=OpacitySphereMask_Radius}
   Hardness=ScalarParameter() {group=Animation, parameterName=OpacitySphereMask_Hardness}
  ) {attenuationRadius=0.5, hardnessPercent=0}
  n26 = NamedRerouteDeclaration(None=LinearInterpolate(A=@n10.A, B=Multiply(A=@n10.A, B=@n25), Alpha=ScalarParameter() {defaultValue=1, group=Animation, parameterName=Opacity_SphereMask})) {name=SphereMask, nodeColor=<rgba 0 0.9297 1 1>}
  NamedRerouteDeclaration(None=LinearInterpolate(B=Multiply(A=@n24, B=@n26), Alpha=ScalarParameter() {defaultValue=1, parameterName=Opacity, sliderMax=1})) {name=AlphaMix, nodeColor=<rgba 0 0.2266 1 1>}
  n27 = NamedRerouteUsage()
  n28 = ScalarParameter() {defaultValue=0.3, parameterName=Roughness Min}
  n29 = ScalarParameter() {defaultValue=0.55, parameterName=Roughness Max}
  n30 = TextureSample() {texture=/Game/Environments/Artemis/Sets/Celebration/Textures/CleanPaint/T_Celebration_Clean_Paint_Background_S}
  n31 = Constant()
  n32 = Constant() {r=0.3}
  n33 = Multiply(
   A=VectorParameter() {parameterName=Decal Tint, defaultValue=<rgba 1 1 1 1>}.RGB
   B=ScalarParameter() {defaultValue=1, group=Animation, parameterName=EmissiveBrightness}
  )
  n34 = Multiply(
   A=StaticSwitchParameter(
    True=MaterialFunctionCall(Color_Morning=@n31, Color_Day=@n31, Color_Evening=@n31, Color_Night=@n32) {materialFunction=&Color_Time_of_Day}.Result
    False=@n31
   ) {defaultValue=true, parameterName=Emissive at night?}
   B=@n33
  )
  n35 = MakeMaterialAttributes!(
   BaseColor=Multiply(A=@n27, B=@n19)
   Metallic=Multiply(A=@n27) {constB=0}
   Specular=Multiply(A=@n27, B=ScalarParameter() {defaultValue=0.5, parameterName=Specular})
   Roughness=Multiply(A=@n27, B=LinearInterpolate(A=@n28, B=@n29, Alpha=@n30.B))
   Anisotropy
   EmissiveColor=Multiply(A=@n27, B=Multiply(A=@n19, B=@n34))
   Opacity=@n27
   OpacityMask
   Normal
   Tangent
   WorldPositionOffset
   SubsurfaceColor
   ClearCoat
   ClearCoatRoughness
   AmbientOcclusion
   Refraction
   CustomizedUVs_0
   CustomizedUVs_1
   CustomizedUVs_2
   CustomizedUVs_3
   CustomizedUVs_4
   CustomizedUVs_5
   CustomizedUVs_6
   CustomizedUVs_7
   PixelDepthOffset
   ShadingModel
   Displacement
  )
  n36 = MaterialFunctionCall!(Material=@n35, Tint=@n33, TintMask) {materialFunction=/Engine/Functions/MaterialLayerFunctions/MatLayerBlend_Tint}
  n37 = VertexColor()
  StaticSwitchParameter(
   True=@n36.Blended Material
   False=StaticSwitchParameter(
    True=MaterialFunctionCall(Material=@n35, Tint=@n37) {materialFunction=&MatLayerBlend_Tint}.Blended Material
    False=@n36.Blended Material
   ) {parameterName=Use Vertex Colors for Tinting}
  ) {parameterName=Use Vertex Color Opacity Masking}
 ]
}
