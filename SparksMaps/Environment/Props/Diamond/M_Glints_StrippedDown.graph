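# Additive, unlit, two-sided glint material for the diamond prop, usable on Niagara mesh particles, that draws flickering, camera-facing sparkles.
# World position offset combines Sprite billboarding with sine-pulsed size, axis rotation and camera offset.
# Emissive layers a center dot, a gradient-mapped corona flickered by multiplied sines phase-shifted by vertex red, and a tinted flare texture.
# Opacity uses flare brightness and depth fade, replaced by a constant on Low and Medium quality.
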Material /SparksMaps/Environment/Props/Diamond/M_Glints_StrippedDown
graph {
 properties {
  bCastRayTracedShadows=false
  bUsedWithNiagaraMeshParticles=true
  blendMode=BLEND_Additive
  pixelDepthOffsetMode=PDOM_Legacy
  shadingModel=MSM_Unlit
  twoSided=true
  usedShadingModels=Unlit
 }
 nodes [
  n1 = Add(A=VertexColor().R, B=Multiply(A=ScalarParameter() {defaultValue=0.2, parameterName=SPeed02}, B=Time()) {constB=0.5})
  Frac(None=@n1)
  n2 = ScalarParameter() {defaultValue=0.2, parameterName=RotationAngle}
  Multiply(A=@n2, B=Time())
  n3 = ScalarParameter() {parameterName=SizeMin}
  n4 = ScalarParameter() {defaultValue=650, parameterName=SizeMax}
  n5 = Sine(None=NamedRerouteUsage())
  n6 = Power(Base=Abs(None=ConstantBiasScale(None=@n5)), Exp=ScalarParameter() {defaultValue=20, parameterName=Pow})
  n7 = Constant() {r=1}
  n8 = ScalarParameter() {defaultValue=1, parameterName=Glint}
  n9 = ScalarParameter() {defaultValue=0.1, parameterName=Radius}
  n10 = MaterialFunctionCall!(
   Center Location=WorldPosition().XYZ
   0-1 UVs=TextureCoordinate()
   XY scale=Multiply(A=LinearInterpolate(A=@n3, B=@n4, Alpha=@n6) {constA=20, constB=80}, B=SphereMask(A=@n7, B=@n8, Radius=@n9) {hardnessPercent=1})
   OverridePivot
   "Normals (Optional)"
  ) {materialFunction=&Sprite}
  n11 = VectorParameter() {parameterName=RotationAxis, defaultValue=<rgba 0 0 1 1>}
  n12 = Transform(None=RotateAboutAxis(NormalizedRotationAxis=Normalize(VectorInput=@n11.RGB), RotationAngle=@n2, PivotPoint=Constant(), Position=LocalPosition().XYZ)) {transformSourceType=TRANSFORMSOURCE_Local}
  n13 = MaterialFunctionCall(Offset Amount=ScalarParameter() {defaultValue=200, parameterName=Camera Offset}, World Position=Add(A=WorldPosition().XYZ, B=@n12)) {materialFunction=/Engine/Functions/Engine_MaterialFunctions02/WorldPositionOffset/CameraOffset}
  NamedRerouteDeclaration(None=Add(A=Add(A=@n10.World position offset, B=@n13.Clamped Camera Offset), B=@n12)) {name=WPO, nodeColor=<rgba 1 0 0.2344 1>}
  n14 = Constant() {r=0.5}
  n15 = SphereMask(A=TextureCoordinate(), B=@n14, Radius=@n14) {hardnessPercent=0}
  n16 = Multiply(A=ScalarParameter() {defaultValue=1, parameterName=CenterBrightness}, B=Power(Base=@n15) {constExponent=25}) {constB=2}
  n17 = Constant() {r=0.5}
  n18 = Multiply(A=Add(A=VertexColor().R, B=Time()), B=ScalarParameter() {defaultValue=1, parameterName=FlickerIntensity})
  n19 = Sine(None=@n18) {period=0.155}
  n20 = Sine(None=@n18) {period=0.2}
  n21 = Multiply(A=Sine(None=@n18) {period=0.075}, B=Multiply(A=Multiply(A=@n19, B=@n20), B=Sine(None=@n18) {period=0.43}))
  n22 = ScalarParameter() {defaultValue=1, parameterName=FlickerLerp}
  n23 = ScalarParameter() {defaultValue=1, parameterName=FlickerFade}
  n24 = LinearInterpolate(Alpha=LinearInterpolate(A=Clamp(None=@n21) {maxDefault=3}, Alpha=Multiply(A=@n22, B=@n23)) {constA=1}) {constA=-0.4, constB=0}
  n25 = ScalarParameter() {defaultValue=1, parameterName=CoronaSize}
  n26 = SphereMask(A=TextureCoordinate(), B=@n17, Radius=@n17, Hardness=LinearInterpolate(Alpha=Add(A=@n24, B=@n25) {constB=0}) {constA=0.7, constB=-0.3}) {hardnessPercent=25}
  n27 = Multiply(
   A=Desaturation(
    None=TextureSampleParameter2D(UVs=@n26) {parameterName=ColorGradient, texture=/Game/Effects/Fort_Effects/Textures/Grad/T_Halo_Corona_Gradient}.RGB
    Fraction=ScalarParameter() {defaultValue=-0.5, parameterName=Desaturate}
   )
   B=ScalarParameter() {defaultValue=0.25, parameterName=CoronaBrightness}
  ) {constB=0.125}
  n28 = Multiply(
   A=VectorParameter() {parameterName=Color, defaultValue=<rgba 0.5 1 1 1>}.RGB
   B=ScalarParameter() {defaultValue=1, parameterName=BrightnessMaster}
  )
  n29 = Multiply(A=ShadingPathSwitch(Default=Add(A=@n16, B=@n27), Forward=@n27), B=LinearInterpolate(B=@n28) {constA=1, constAlpha=0.7})
  n30 = Add(
   A=Multiply(A=Power(Base=@n15) {constExponent=4}, B=ScalarParameter() {defaultValue=0.2, parameterName=EdgeBrightness}) {constB=0.15}
   B=TextureSampleParameter2D() {parameterName=FlareMask, texture=&/T_X_Flare_ColorsRGB}.RGB
  )
  n31 = Multiply(A=@n30, B=@n28)
  NamedRerouteDeclaration(None=MaterialFunctionCall(Emissive=Add(A=@n29, B=@n31)) {materialFunction=&MF_FX_EmissiveAdjustments}.Result) {name=Emissive, nodeColor=<rgba 1 0.9249 0.9671 1>}
  n32 = ScalarParameter() {defaultValue=0.2, parameterName=AlphaPow}
  n33 = Multiply(
   A=Multiply(A=@n28, B=Power(Base=@n30, Exp=@n32))
   B=DepthFade(FadeDistance=ScalarParameter() {defaultValue=50, parameterName=DepthFade_Distance})
  )
  n34 = Constant() {r=1}
  NamedRerouteDeclaration(None=Multiply(A=QualitySwitch(Default=@n33, Low=@n34, High=@n33, Medium=@n34, Epic=@n33), B=ScalarParameter() {defaultValue=1, parameterName=Alpha})) {name=Opacity, nodeColor=<rgba 1 0 0.6797 1>}
  NamedRerouteDeclaration(None=@n1) {name=RandomPanValue, nodeColor=<rgba 0 1 0.9766 1>}
  NamedRerouteUsage()
  NamedRerouteUsage()
  NamedRerouteUsage()
  ScalarParameter() {parameterName=RotationSpeed}
 ]
}
